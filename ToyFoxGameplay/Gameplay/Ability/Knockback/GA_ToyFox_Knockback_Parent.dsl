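; Parent gameplay ability for knockbacks in ToyFox gameplay that extends an inherited launch step.
; Its DoLaunch override first runs the parent launch with the given velocity.
; Then, on authority only and if the fall damage immunity flag is set, it applies an Athena gameplay effect to the owner granting fall damage immunity until landing, in a no-FX variant.
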(event EventDoLaunch
  (params (Vel :type Vector :default "0, 0, 0"))
  ;[0 0.15 0.0122 0.5] DoLaunch: Provide fall damage immunity if desired.
  (|CallParentFunction Vel)
  (Ability|HasAuthority
    (:True
      (if (Variables|Default|GetbProvideFallDamageImmunity)
        (Ability|ApplyGameplayEffectToOwner "/Game/Athena/Items/Parents/FallImmunity/GE_Athena_TillLandFallDamageImmunity_NoFX_Grant_Parent.GE_Athena_TillLandFallDamageImmunity_NoFX_Grant_Parent_C")))))
